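# Physics asset for the M_MED_Candor_Dark backpack cosmetic, defining secondary motion for its straps, strap rings and zippers.
# A large sphere on the pack bone anchors small sphere bodies on the dyn_ bones.
# Constraints form two-link chains from the pack: left and right zippers, rings, side straps and upper/lower holder straps, plus a left-only center strap.
# Tight swing and twist limits restrict each link to small movements.
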PhysicsAsset /BRCosmetics/Accessories/FORT_Backpacks/Backpack_M_MED_Candor_Dark/Meshes/M_MED_Candor_Dark_Pack_Physics
bodies {
 dyn_center_strap_left_1 [{shapeType=Sphere, radius=1.25, shapeName=dyn_center_strap_left_1_sphere, center=<0.0392 -0.0045 -0.1022>}]
 dyn_center_strap_left_2 [{shapeType=Sphere, radius=1.25, shapeName=dyn_center_strap_left_2_sphere, center=<0.5526 -0.0197 -0.002>}]
 dyn_side_lower_holder_strap_left_1 [{shapeType=Sphere, radius=0.75, shapeName=dyn_side_lower_holder_strap_left_1_sphere}]
 dyn_side_lower_holder_strap_left_2 [{shapeType=Sphere, radius=0.75, shapeName=dyn_side_lower_holder_strap_left_2_sphere, center=<-0.1039 0.367 0.1486>}]
 dyn_side_lower_holder_strap_right_1 [{shapeType=Sphere, radius=0.75, shapeName=dyn_side_lower_holder_strap_right_1_sphere}]
 dyn_side_lower_holder_strap_right_2 [{shapeType=Sphere, radius=0.75, shapeName=dyn_side_lower_holder_strap_right_2_sphere, center=<0.1306 -0.3278 -0.1486>}]
 dyn_side_strap_left_1 [{shapeType=Sphere, radius=2, shapeName=dyn_side_strap_left_1_sphere, center=<2.3289 0.0268 0.3369>}]
 dyn_side_strap_left_base [{shapeType=Sphere, radius=2, shapeName=dyn_side_strap_left_base_sphere, center=<3.0767 1.195 -0.1927>}]
 dyn_side_strap_right_1 [{shapeType=Sphere, radius=2, shapeName=dyn_side_strap_right_1_sphere, center=<2.3273 0.0207 0.3369>}]
 dyn_side_strap_right_base [{shapeType=Sphere, radius=2, shapeName=dyn_side_strap_right_base_sphere, center=<3.0189 -0.1472 -0.2086>}]
 dyn_side_upper_holder_strap_left_1 [{shapeType=Sphere, radius=0.75, shapeName=dyn_side_upper_holder_strap_left_1_sphere, center=<-0.037 0.0077 0.0053>}]
 dyn_side_upper_holder_strap_left_2 [{shapeType=Sphere, radius=0.75, shapeName=dyn_side_upper_holder_strap_left_2_sphere, center=<-0.4521 0.6549 -0.044>}]
 dyn_side_upper_holder_strap_right_1 [{shapeType=Sphere, radius=0.75, shapeName=dyn_side_upper_holder_strap_right_1_sphere, center=<0.0735 0.0218 0.0401>}]
 dyn_side_upper_holder_strap_right_2 [{shapeType=Sphere, radius=0.75, shapeName=dyn_side_upper_holder_strap_right_2_sphere, center=<0.4684 -0.6103 0.044>}]
 dyn_strap_ring_left_1 [{shapeType=Sphere, radius=1.25, shapeName=dyn_strap_ring_left_1_sphere, center=<0.9437 -0.0892 -0.1414>}]
 dyn_strap_ring_left_2 [{shapeType=Sphere, radius=1.25, shapeName=dyn_strap_ring_left_2_sphere, center=<0.4823 0.0001 0.1546>}]
 dyn_strap_ring_right_1 [{shapeType=Sphere, radius=1.25, shapeName=dyn_strap_ring_right_1_sphere, center=<1.8932 0.1264 -0.2486>}]
 dyn_strap_ring_right_2 [{shapeType=Sphere, radius=1.25, shapeName=dyn_strap_ring_right_2_sphere, center=<0.4775 0.0365 0.127>}]
 dyn_zipper_left_1 [{shapeType=Sphere, radius=0.5, shapeName=dyn_zipper_left_1_sphere, center=<0.5267 0.0567 0.0403>}]
 dyn_zipper_left_2 [{shapeType=Sphere, radius=1.25, shapeName=dyn_zipper_left_2_sphere, center=<2.4449 0.0593 -0.0213>}]
 dyn_zipper_right_1 [{shapeType=Sphere, radius=0.5, shapeName=dyn_zipper_right_1_sphere, center=<0.5267 -0.0566 0.0397>}]
 dyn_zipper_right_2 [{shapeType=Sphere, radius=1.25, shapeName=dyn_zipper_right_2_sphere, center=<2.4449 -0.0593 -0.0226>}]
 pack [{shapeType=Sphere, radius=12.2396, shapeName=pack_sphere, center=<0.0237 -5.6686 -10.9126>}]
}
constraints [
 {bone1Name=dyn_zipper_left_1, bone2Name=pack, swing1LimitDegrees=30, swing2LimitDegrees=10, twistLimitDegrees=5}
 {bone1Name=dyn_zipper_left_2, bone2Name=dyn_zipper_left_1, swing1LimitDegrees=30, swing2LimitDegrees=10, twistLimitDegrees=5}
 {bone1Name=dyn_strap_ring_left_1, bone2Name=pack, swing1LimitDegrees=10, swing2LimitDegrees=10, twistLimitDegrees=5}
 {bone1Name=dyn_strap_ring_left_2, bone2Name=dyn_strap_ring_left_1, swing1LimitDegrees=10, swing2LimitDegrees=10, twistLimitDegrees=5}
 {bone1Name=dyn_center_strap_left_1, bone2Name=pack, swing1LimitDegrees=15, swing2LimitDegrees=10, twistLimitDegrees=5}
 {bone1Name=dyn_center_strap_left_2, bone2Name=dyn_center_strap_left_1, swing1LimitDegrees=15, swing2LimitDegrees=10, twistLimitDegrees=5}
 {bone1Name=dyn_zipper_right_1, bone2Name=pack, swing1LimitDegrees=15, swing2LimitDegrees=10, twistLimitDegrees=5}
 {bone1Name=dyn_zipper_right_2, bone2Name=dyn_zipper_right_1, swing1LimitDegrees=30, swing2LimitDegrees=10, twistLimitDegrees=5}
 {bone1Name=dyn_strap_ring_right_1, bone2Name=pack, swing1LimitDegrees=10, swing2LimitDegrees=10, twistLimitDegrees=5}
 {bone1Name=dyn_strap_ring_right_2, bone2Name=dyn_strap_ring_right_1, swing1LimitDegrees=10, swing2LimitDegrees=10, twistLimitDegrees=5}
 {bone1Name=dyn_side_strap_right_1, bone2Name=dyn_side_strap_right_base, swing1LimitDegrees=15, swing2LimitDegrees=10, twistLimitDegrees=5}
 {bone1Name=dyn_side_strap_right_base, bone2Name=pack, swing1LimitDegrees=15, swing2LimitDegrees=10, twistLimitDegrees=5}
 {bone1Name=dyn_side_strap_left_1, bone2Name=dyn_side_strap_left_base, swing1LimitDegrees=15, swing2LimitDegrees=10, twistLimitDegrees=5}
 {bone1Name=dyn_side_strap_left_base, bone2Name=pack, swing1LimitDegrees=15, swing2LimitDegrees=10, twistLimitDegrees=5}
 {bone1Name=dyn_side_upper_holder_strap_right_1, bone2Name=pack, swing1LimitDegrees=15, swing2LimitDegrees=10, twistLimitDegrees=5}
 {
  bone1Name=dyn_side_upper_holder_strap_right_2
  bone2Name=dyn_side_upper_holder_strap_right_1
  swing1LimitDegrees=15
  swing2LimitDegrees=10
  twistLimitDegrees=5
 }
 {bone1Name=dyn_side_lower_holder_strap_right_1, bone2Name=pack, swing1LimitDegrees=15, swing2LimitDegrees=10, twistLimitDegrees=5}
 {
  bone1Name=dyn_side_lower_holder_strap_right_2
  bone2Name=dyn_side_lower_holder_strap_right_1
  swing1LimitDegrees=15
  swing2LimitDegrees=10
  twistLimitDegrees=5
 }
 {bone1Name=dyn_side_upper_holder_strap_left_1, bone2Name=pack, swing1LimitDegrees=15, swing2LimitDegrees=10, twistLimitDegrees=5}
 {
  bone1Name=dyn_side_upper_holder_strap_left_2
  bone2Name=dyn_side_upper_holder_strap_left_1
  swing1LimitDegrees=15
  swing2LimitDegrees=10
  twistLimitDegrees=5
 }
 {bone1Name=dyn_side_lower_holder_strap_left_1, bone2Name=pack, swing1LimitDegrees=15, swing2LimitDegrees=10, twistLimitDegrees=5}
 {
  bone1Name=dyn_side_lower_holder_strap_left_2
  bone2Name=dyn_side_lower_holder_strap_left_1
  swing1LimitDegrees=15
  swing2LimitDegrees=10
  twistLimitDegrees=5
 }
]
