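# Animation Blueprint for the Candor Dark backpack cosmetic, adding physics motion and a strap adjustment while its owner moves vertically.
# It copies the attached parent's pose from Backpack_BR, layers reference pose from root down, then runs RigidBody simulation with overridden gravity and damped component motion.
# Additive rotations on both side-strap base bones follow JumpingAlpha, which the update event snaps to 1 whenever pawn Z velocity is nonzero, otherwise 0.
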AnimBlueprint /BRCosmetics/Accessories/FORT_Backpacks/Backpack_M_MED_Candor_Dark/Meshes/M_MED_Candor_Dark_Pack_AnimBP
target_skeleton=&/M_MED_Candor_Dark_Pack_Skeleton
graphs [
 AnimGraph {nodes [
   Input Pose()
   n1 = Copy Pose From Mesh() {bCopyCurves=true, bUseAttachedParent=true, rootBoneToCopy=Backpack_BR}
   n2 = Local Space Ref Pose()
   n3 = RigidBody(ComponentPose=Local To Component(LocalPose=Layered blend per bone(BasePose=@n1, BlendPoses_0=@n2) {layerSetup [{branchFilters [{boneName=root}]}]})) {
    bOverrideWorldGravity=true
    overridePhysicsAsset=&/M_MED_Candor_Dark_Pack_Physics
    alphaBoolBlend {blendOption=Linear}
    componentAppliedLinearAccClamp=<3000 3000 3000>
    componentLinearAccScale=<0.3 0.3 0.3>
    componentLinearVelScale=<0.3 0.3 0.3>
    overrideWorldGravity {z=-980}
   }
   n4 = Get JumpingAlpha() -> (JumpingAlpha:double)
   n5 = "Transform (Modify) Bone - Bone: dyn_side_strap_right_base"(ComponentPose=@n3, Alpha=@n4, Translation="0,0,0", Rotation="-2,0,0", Scale="1,1,1") {rotationMode=BMM_Additive, alphaBoolBlend {blendOption=Linear}, boneToModify {boneName=dyn_side_strap_right_base}, rotation {pitch=-2}}
   n6 = "Transform (Modify) Bone - Bone: dyn_side_strap_left_base"(ComponentPose=@n5, Alpha=@n4, Translation="0,0,0", Rotation="11,0.000,0.000", Scale="1,1,1") {rotationMode=BMM_Additive, alphaBoolBlend {blendOption=Linear}, boneToModify {boneName=dyn_side_strap_left_base}, rotation {pitch=11}}
   Output Pose(Result=Component To Local(ComponentPose=@n6)) {layerGroup=None}
  ]}
 EventGraph {nodes [
   Event BlueprintUpdateAnimation()
   n1 = TryGetPawnOwner()
   > n2 = Branch(Condition="Not Equal (Float)"(A=GetVelocity!(self:Actor Object Reference=@n1) -> (ReturnValue_X:double, ReturnValue_Y:double, ReturnValue_Z:double).ReturnValue_Z))
   > Set JumpingAlpha(JumpingAlpha:double=1) -> (Output_Get:double)
   Set JumpingAlpha(execute=@n2.else, JumpingAlpha=0)
  ]}
]
variables [JumpingAlpha {type=number}]
